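# Sound cue for the bounce impact of the proximity grenade launcher's grenades.
# A random node picks one of three equally weighted bounce waves and feeds it through a modulator with pitch pinned at 1.
# It plays in the Fort_Impacts sound class with grenade-bounce attenuation.
# A concurrency override caps simultaneous instances at three and prevents new ones beyond that.
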SoundCue /Game/Sounds/Fort_Weapons/Launchers/Prox_GrenadeLauncher/Weapon_LAU_GrenadeLauncher_Prox_Bounce_Cue
root=Modulator_5
nodes [
 {id=Modulator_5, children [Random_0], properties {pitchMax=1, pitchMin=1}}
 {id=Random_0, children [WavePlayer_3,WavePlayer_4,WavePlayer_5], properties {weights [1,1,1]}}
 {id=WavePlayer_3, properties {soundWaveAssetPtr=&/Proximity_GrenadeLauncher_Bounce_01}}
 {id=WavePlayer_4, properties {soundWaveAssetPtr=&/Proximity_GrenadeLauncher_Bounce_02}}
 {id=WavePlayer_5, properties {soundWaveAssetPtr=&/Proximity_GrenadeLauncher_Bounce_03}}
]
properties {
 attenuationSettings=/Game/Sounds/Attenuation/Fort_Grenade_Bounce_Att
 bOverrideConcurrency=true
 maxDistance=5500
 soundClassObject=&Fort_Impacts
 volumeMultiplier=1
 concurrencyOverrides {maxCount=3, resolutionRule=PreventNew}
}
