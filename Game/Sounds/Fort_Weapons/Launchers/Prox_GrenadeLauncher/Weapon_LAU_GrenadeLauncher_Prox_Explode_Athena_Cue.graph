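# Sound cue for the proximity grenade launcher's explosion in Athena, layered by distance.
# After a short random delay, a mixer blends close, medium and distant layers, each randomly picking one of five waves and using its own distance attenuation with overridden falloff.
# Modulators vary close-layer volume and medium and distant pitch, while sound class overrides route the distant layer separately.
# A concurrency override stops the oldest instance at the limit.
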SoundCue /Game/Sounds/Fort_Weapons/Launchers/Prox_GrenadeLauncher/Weapon_LAU_GrenadeLauncher_Prox_Explode_Athena_Cue
root=Delay_1
nodes [
 {
  id=Attenuation_0
  children [Random_3]
  properties {attenuationSettings=/Game/Sounds/Attenuation/Fort_Explosion_Medium_Athena_Att, attenuationOverrides {falloffDistance=1600}}
 }
 {
  id=Attenuation_5
  children [Modulator_5]
  properties {
   attenuationSettings=/Game/Sounds/Attenuation/Fort_Explosion_Distant_Athena_Att
   attenuationOverrides {falloffDistance=3000, attenuationShapeExtents {x=1000}}
  }
 }
 {
  id=Attenuation_6
  children [Random_0]
  properties {attenuationSettings=/Game/Sounds/Attenuation/Fort_Explosion_Close_Athena_Att, attenuationOverrides {falloffDistance=1600}}
 }
 {id=Delay_1, children [Mixer_4], properties {delayMax=0.1, delayMin=0.05}}
 {id=Mixer_4, children [SoundClass_6,SoundClass_2,SoundClass_0], properties {inputVolume [0.95,0.65,1]}}
 {id=Modulator_0, children [Attenuation_0], properties {pitchMax=1.15, pitchMin=1}}
 {id=Modulator_5, children [Random_2], properties {pitchMax=1.1, pitchMin=0.9}}
 {id=Modulator_6, children [Attenuation_6], properties {volumeMax=1.1, volumeMin=0.9}}
 {id=Random_0, children [WavePlayer_0,WavePlayer_1,WavePlayer_2,WavePlayer_3,WavePlayer_4], properties {weights [1,1,1,1,1]}}
 {id=Random_2, children [WavePlayer_5, WavePlayer_6, WavePlayer_12, WavePlayer_13, WavePlayer_14], properties {weights [1,1,1,1,1]}}
 {id=Random_3, children [WavePlayer_15, WavePlayer_16, WavePlayer_17, WavePlayer_18, WavePlayer_19], properties {weights [1,1,1,1,1]}}
 {id=SoundClass_0, children [Modulator_0], properties {soundClassOverride=&Fort_WeaponRanged_Fire_3P_PSM}}
 {
  id=SoundClass_2
  children [Attenuation_5]
  properties {soundClassOverride=&Fort_WeaponRanged_Fire_Distant}
 }
 {id=SoundClass_6, children [Modulator_6], properties {soundClassOverride=&Fort_WeaponRanged_Fire_3P_PSM}}
 {
  id=WavePlayer_0
  properties {soundWaveAssetPtr=&/Proximity_GrenadeLauncher_Explo_Close_01}
 }
 {
  id=WavePlayer_1
  properties {soundWaveAssetPtr=&/Proximity_GrenadeLauncher_Explo_Close_02}
 }
 {
  id=WavePlayer_12
  properties {soundWaveAssetPtr=&/Proximity_GrenadeLauncher_Explo_Dist_03}
 }
 {
  id=WavePlayer_13
  properties {soundWaveAssetPtr=&/Proximity_GrenadeLauncher_Explo_Dist_04}
 }
 {
  id=WavePlayer_14
  properties {soundWaveAssetPtr=&/Proximity_GrenadeLauncher_Explo_Dist_05}
 }
 {
  id=WavePlayer_15
  properties {soundWaveAssetPtr=&/Proximity_GrenadeLauncher_Explo_Med_01}
 }
 {
  id=WavePlayer_16
  properties {soundWaveAssetPtr=&/Proximity_GrenadeLauncher_Explo_Med_02}
 }
 {
  id=WavePlayer_17
  properties {soundWaveAssetPtr=&/Proximity_GrenadeLauncher_Explo_Med_03}
 }
 {
  id=WavePlayer_18
  properties {soundWaveAssetPtr=&/Proximity_GrenadeLauncher_Explo_Med_04}
 }
 {
  id=WavePlayer_19
  properties {soundWaveAssetPtr=&/Proximity_GrenadeLauncher_Explo_Med_05}
 }
 {
  id=WavePlayer_2
  properties {soundWaveAssetPtr=&/Proximity_GrenadeLauncher_Explo_Close_03}
 }
 {
  id=WavePlayer_3
  properties {soundWaveAssetPtr=&/Proximity_GrenadeLauncher_Explo_Close_04}
 }
 {
  id=WavePlayer_4
  properties {soundWaveAssetPtr=&/Proximity_GrenadeLauncher_Explo_Close_05}
 }
 {
  id=WavePlayer_5
  properties {soundWaveAssetPtr=&/Proximity_GrenadeLauncher_Explo_Dist_01}
 }
 {
  id=WavePlayer_6
  properties {soundWaveAssetPtr=&/Proximity_GrenadeLauncher_Explo_Dist_02}
 }
]
properties {bOverrideConcurrency=true, maxDistance=27500, volumeMultiplier=1, concurrencyOverrides {resolutionRule=StopOldest, volumeScale=0.5}}
