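# First-person firing sound cue for the Proximity Grenade Launcher in Athena.
# A Random node picks among four equally weighted close-range fire waves and is excluded from branch culling.
# It feeds a Modulator that holds pitch fixed, leaving only default volume variation.
# The cue uses first-person weapon attenuation, the ranged-fire 1P sound class and a first-person gunfire concurrency set.
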SoundCue /Game/Sounds/Fort_Weapons/Launchers/Prox_GrenadeLauncher/Weapon_LAU_GrenadeLauncher_Prox_Fire_1P_Athena_Cue
root=Modulator_6
nodes [
 {id=Modulator_6, children [Random_0], properties {pitchMax=1, pitchMin=1}}
 {
  id=Random_0
  children [WavePlayer_0,WavePlayer_1,WavePlayer_2,WavePlayer_3]
  properties {bShouldExcludeFromBranchCulling=true, weights [1,1,1,1]}
 }
 {
  id=WavePlayer_0
  properties {soundWaveAssetPtr=&/Proximity_GrenadeLauncher_Fire_Close_Alt_01}
 }
 {
  id=WavePlayer_1
  properties {soundWaveAssetPtr=&/Proximity_GrenadeLauncher_Fire_Close_Alt_02}
 }
 {
  id=WavePlayer_2
  properties {soundWaveAssetPtr=&/Proximity_GrenadeLauncher_Fire_Close_Alt_03}
 }
 {
  id=WavePlayer_3
  properties {soundWaveAssetPtr=&/Proximity_GrenadeLauncher_Fire_Close_Alt_04}
 }
]
properties {
 attenuationSettings=&Fort_Weapon_Projectile_FirstPerson_Athena_Att
 maxDistance=2097152
 soundClassObject=&Fort_WeaponRanged_Fire_1P_PSM
 volumeMultiplier=1.2
 attenuationOverrides {falloffDistance=5000, attenuationShapeExtents {x=1000}}
 concurrencyOverrides {maxCount=3}
 concurrencySet [&Weapon_Gunfire_FirstPerson_Conc]
}
