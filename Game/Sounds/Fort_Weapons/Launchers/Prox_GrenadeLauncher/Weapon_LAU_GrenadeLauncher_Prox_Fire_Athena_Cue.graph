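# Athena fire sound cue for the proximity grenade launcher, layering close, medium and distant recordings.
# A mixer blends the three layers, with the distant one much quieter, each picking randomly from four waves under its own attenuation.
# Sound class overrides put close on the third-person fire class and the others on the distant class; the distant layer also gets pitch variation.
# Global and local gunfire concurrency sets are assigned.
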SoundCue /Game/Sounds/Fort_Weapons/Launchers/Prox_GrenadeLauncher/Weapon_LAU_GrenadeLauncher_Prox_Fire_Athena_Cue
root=Mixer_3
nodes [
 {
  id=Attenuation_10
  children [Random_2]
  properties {attenuationSettings=&Fort_Weapon_Projectile_Medium_Mono_Athena_Att}
 }
 {
  id=Attenuation_11
  children [Modulator_7]
  properties {attenuationSettings=&Fort_Weapon_Projectile_Distant_Mono_Athena_Att}
 }
 {id=Attenuation_9, children [Modulator_6], properties {attenuationSettings=&Fort_Weapon_Projectile_Close_Athena_Att}}
 {id=Mixer_3, children [SoundClass_1,SoundClass_2,SoundClass_6], properties {inputVolume [1,0.9,0.3]}}
 {id=Modulator_6, children [Random_0], properties {pitchMax=1, pitchMin=1}}
 {id=Modulator_7, children [Random_1], properties {pitchMax=1.1}}
 {
  id=Random_0
  children [WavePlayer_0,WavePlayer_1,WavePlayer_2,WavePlayer_3]
  properties {bShouldExcludeFromBranchCulling=true, weights [1,1,1,1]}
 }
 {
  id=Random_1
  children [WavePlayer_4,WavePlayer_5,WavePlayer_6,WavePlayer_7]
  properties {bShouldExcludeFromBranchCulling=true, weights [1,1,1,1]}
 }
 {
  id=Random_2
  children [WavePlayer_8, WavePlayer_9, WavePlayer_10, WavePlayer_11]
  properties {bShouldExcludeFromBranchCulling=true, weights [1,1,1,1]}
 }
 {
  id=SoundClass_1
  children [Attenuation_9]
  properties {soundClassOverride=&Fort_WeaponRanged_Fire_3P_PSM}
 }
 {
  id=SoundClass_2
  children [Attenuation_10]
  properties {soundClassOverride=&Fort_WeaponRanged_Fire_Distant}
 }
 {
  id=SoundClass_6
  children [Attenuation_11]
  properties {soundClassOverride=&Fort_WeaponRanged_Fire_Distant}
 }
 {
  id=WavePlayer_0
  properties {soundWaveAssetPtr=&/Proximity_GrenadeLauncher_Fire_Close_Alt_01}
 }
 {
  id=WavePlayer_1
  properties {soundWaveAssetPtr=&/Proximity_GrenadeLauncher_Fire_Close_Alt_02}
 }
 {
  id=WavePlayer_10
  properties {soundWaveAssetPtr=&/Proximity_GrenadeLauncher_Fire_Med_Alt_03}
 }
 {
  id=WavePlayer_11
  properties {soundWaveAssetPtr=&/Proximity_GrenadeLauncher_Fire_Med_Alt_04}
 }
 {
  id=WavePlayer_2
  properties {soundWaveAssetPtr=&/Proximity_GrenadeLauncher_Fire_Close_Alt_03}
 }
 {
  id=WavePlayer_3
  properties {soundWaveAssetPtr=&/Proximity_GrenadeLauncher_Fire_Close_Alt_04}
 }
 {
  id=WavePlayer_4
  properties {soundWaveAssetPtr=&/Proximity_GrenadeLauncher_Fire_Dist_Alt_01}
 }
 {
  id=WavePlayer_5
  properties {soundWaveAssetPtr=&/Proximity_GrenadeLauncher_Fire_Dist_Alt_02}
 }
 {
  id=WavePlayer_6
  properties {soundWaveAssetPtr=&/Proximity_GrenadeLauncher_Fire_Dist_Alt_03}
 }
 {
  id=WavePlayer_7
  properties {soundWaveAssetPtr=&/Proximity_GrenadeLauncher_Fire_Dist_Alt_04}
 }
 {
  id=WavePlayer_8
  properties {soundWaveAssetPtr=&/Proximity_GrenadeLauncher_Fire_Med_Alt_01}
 }
 {
  id=WavePlayer_9
  properties {soundWaveAssetPtr=&/Proximity_GrenadeLauncher_Fire_Med_Alt_02}
 }
]
properties {
 maxDistance=27500
 soundClassObject=&Fort_WeaponRanged_Fire_3P_PSM
 volumeMultiplier=1
 attenuationOverrides {falloffDistance=5000, attenuationShapeExtents {x=1000}}
 concurrencyOverrides {maxCount=3}
 concurrencySet [&Weapon_Gunfire_Global_Conc, &Weapon_Gunfire_Local_Conc]
}
