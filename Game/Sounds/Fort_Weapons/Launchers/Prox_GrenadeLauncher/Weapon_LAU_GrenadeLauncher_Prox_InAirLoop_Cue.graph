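# Sound cue for the looping in-air sound of a proximity grenade launcher projectile.
# A looping wave player feeds a Doppler node with raised intensity, so pitch shifts as the grenade moves relative to the listener.
# The cue uses a shared projectile in-air attenuation, a third-person ranged weapon fire sound class and reduced volume, with overridden concurrency capping simultaneous instances.
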SoundCue /Game/Sounds/Fort_Weapons/Launchers/Prox_GrenadeLauncher/Weapon_LAU_GrenadeLauncher_Prox_InAirLoop_Cue
root=Doppler_0
nodes [
 {id=Doppler_0, children [WavePlayer_0], properties {dopplerIntensity=2}}
 {
  id=WavePlayer_0
  properties {bLooping=true, soundWaveAssetPtr=&/Proximity_GrenadeLauncher_InAir_Loop}
 }
]
properties {
 attenuationSettings=/Game/Sounds/Attenuation/Fort_Weapon_Projectile_InAir_Loop_Att
 bOverrideConcurrency=true
 maxDistance=4500
 soundClassObject=&Fort_WeaponRanged_Fire_3P_PSM
 volumeMultiplier=0.5
 concurrencyOverrides {maxCount=3}
}
