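# Sound cue for picking up the proximity grenade launcher in Athena, classed as a loot-collect sound.
# A quality-level switch routes its first two levels to one mixer and its last two to another.
# Each mixer boosts a single wave player, both referencing a Pickup_01 wave of the same name, and leaves its second input empty.
# Weapon-pickup attenuation spatializes the cue, and the shared PickupMaximum_Oldest concurrency set governs overlapping plays.
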SoundCue /Game/Sounds/Fort_Weapons/Launchers/Prox_GrenadeLauncher/Weapon_LAU_GrenadeLauncher_Prox_Pickup_Athena_Cue
root=QualityLevel_1
nodes [
 {id=Mixer_0, children [WavePlayer_7,null], properties {inputVolume [1.6,1]}}
 {id=Mixer_1, children [WavePlayer_8,null], properties {inputVolume [1.6,1]}}
 {id=QualityLevel_1, children [Mixer_0,Mixer_0,Mixer_1,Mixer_1]}
 {id=WavePlayer_7, properties {soundWaveAssetPtr=&/Proximity_GrenadeLauncher_Pickup_01}}
 {id=WavePlayer_8, properties {soundWaveAssetPtr=&Proximity_GrenadeLauncher_Pickup_01}}
]
properties {
 attenuationSettings=&Fort_Weapon_Pickup_Att
 maxDistance=3500
 soundClassObject=&Fort_Loot_Collect
 volumeMultiplier=0.3231
 concurrencyOverrides {maxCount=1, resolutionRule=PreventNew}
 concurrencySet [&PickupMaximum_Oldest]
}
