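# Sound Cue for the projectile-by sound of the proximity grenade launcher, routed to a third-person ranged weapon fire sound class.
# It picks one of three equally weighted waves at random, then passes it through a modulator with pitch variation disabled.
# An attenuation node referencing the RPG explosion settings carries overrides for falloff and distance-based low-pass filtering.
# The cue reduces overall volume and overrides concurrency to prevent new instances at the limit.
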SoundCue /Game/Sounds/Fort_Weapons/Launchers/Prox_GrenadeLauncher/Weapon_LAU_GrenadeLauncher_Prox_ProjectileBy_Cue
root=Attenuation_9
nodes [
 {
  id=Attenuation_9
  children [Modulator_6]
  properties {
   attenuationSettings=/Game/Sounds/Attenuation/Fort_Explosion_RPG_Att
   attenuationOverrides {bAttenuateWithLPF=true, falloffDistance=3000, lPFRadiusMax=3000, lPFRadiusMin=1500, attenuationShapeExtents {x=2000}}
  }
 }
 {id=Modulator_6, children [Random_0], properties {pitchMax=1, pitchMin=1}}
 {id=Random_0, children [WavePlayer_1,WavePlayer_2,WavePlayer_3], properties {weights [1,1,1]}}
 {
  id=WavePlayer_1
  properties {soundWaveAssetPtr=&/Proximity_GrenadeLauncher_ProjectileBy_01}
 }
 {
  id=WavePlayer_2
  properties {soundWaveAssetPtr=&/Proximity_GrenadeLauncher_ProjectileBy_02}
 }
 {
  id=WavePlayer_3
  properties {soundWaveAssetPtr=&/Proximity_GrenadeLauncher_ProjectileBy_03}
 }
]
properties {
 bOverrideConcurrency=true
 maxDistance=13000
 soundClassObject=&Fort_WeaponRanged_Fire_3P_PSM
 volumeMultiplier=0.7
 attenuationOverrides {falloffDistance=5000, attenuationShapeExtents {x=1000}}
 concurrencyOverrides {resolutionRule=PreventNew}
}
